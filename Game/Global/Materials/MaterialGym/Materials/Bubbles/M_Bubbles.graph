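# Unlit material in the MaterialGym folder that displays the bubble effect from the shared MF_Bubbles_OUT function on its own.
# The function's inputs are left unwired, and its Bubble_OUT output is multiplied by itself, squaring it.
# The Emissive_RE color parameter, black by default, is added to that result.
# The sum drives Emissive Color through Set Material Attributes, the only attribute the material sets.
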Material /Game/Global/Materials/MaterialGym/Materials/Bubbles/M_Bubbles
graph {
 properties {bUseMaterialAttributes=true, pixelDepthOffsetMode=PDOM_Legacy, shadingModel=MSM_Unlit, usedShadingModels=Unlit}
 nodes [
  n1 = MaterialFunctionCall!(Material, BubblesMask, SlurpNormalDistortionMask, Pre Skinned Position, TechArtMask, UseBubble, UseTriPlanarBubbles) {materialFunction=&/MF_Bubbles_OUT}
  SetMaterialAttributes(Emissive Color=Add(A=Multiply(A=@n1.Bubble_OUT, B=@n1.Bubble_OUT), B=VectorParameter() {parameterName=Emissive_RE, defaultValue=<rgba 0 0 0 1>}.RGB)) {attributeSetTypes [#1]}
 ]
}
